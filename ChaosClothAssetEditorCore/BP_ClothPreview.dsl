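; Preview actor Blueprint for Chaos Cloth assets, showing Dataflow cloth alongside animated skeletal meshes.
; The construction script assigns the Dataflow asset to the cloth component and attaches it to the base skeletal mesh when valid.
; It adds configured extra meshes, using override or shared animation, and hides whole meshes or chosen material sections across all LODs.
; When Dataflow simulation toggles, optional blend-weight preview switches cloth blend between the set value and zero.
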(event EventBeginPlay)

(event Collision|EventActorBeginOverlap
  (params (OtherActor :type Actor%20Object%20Reference)))

(event EventTick
  (params (DeltaSeconds :type Float%20%28single-precision%29)))

(event Dataflow|EventOnDataflowSimulationEnabledChanged
  (params (bSimulationEnabled :type Boolean))
  (bind _chaosclothcomponent (Variables|Default|GetChaosClothComponent))
  ; Blend Weight control for the sim mesh when Dataflow is stopped and started
  (if (Variables|Default|GetPreviewBlendWeight)
    (if bSimulationEnabled
      (bind _output_get (Variables|/Script/ChaosClothAssetEngine.ChaosClothComponent|SetBlendWeight (Variables|Default|GetBlendWeight) _chaosclothcomponent))
      (Variables|/Script/Engine.SkeletalMeshComponent|SetClothBlendWeight _output_get (Variables|Default|GetSkeletalMeshComponent))
      (else
        (bind _output_get_2 (Variables|/Script/ChaosClothAssetEngine.ChaosClothComponent|SetBlendWeight 0.0 _chaosclothcomponent))
        (Variables|/Script/Engine.SkeletalMeshComponent|SetClothBlendWeight _output_get_2 (Variables|Default|GetSkeletalMeshComponent))))))

(fn ConstructionScript
  (bind _skeletalmeshcomponent (Variables|Default|GetSkeletalMeshComponent))
  (bind _skelmeshcomp_local (LocalVariables|GetSkelMeshComp%20Local))
  (Utilities|FlowControl|Sequence
    (:then_0
      ; Base SkelMesh for the Dataflow
      (ClothComponent|SetAsset (Variables|Default|GetChaosClothComponent) (Variables|Default|GetDataflowAsset))
      (if (Utilities|IsValid (Variables|PreviewSkelMesh|GetSkeletalMesh) :!class /Script/Engine.KismetSystemLibrary)
        (Transformation|AttachComponentToComponent (Variables|Default|GetChaosClothComponent) (Variables|Default|GetSkeletalMeshComponent) "None" "KeepRelative" "KeepRelative" "KeepRelative" true)
        (Variables|/Script/Engine.SkeletalMeshComponent|SetSkeletalMeshAsset (Variables|PreviewSkelMesh|GetSkeletalMesh) (Variables|Default|GetSkeletalMeshComponent))
        (Components|Animation|OverrideAnimationData (Variables|Default|GetSkeletalMeshComponent) (Variables|PreviewSkelMesh|GetAnimationAsset) true true 0.0 1.0)
        ;; Set Num LODs/Materials
        (LocalVariables|SetLastLOD (- (Components|SkinnedMesh|GetNumLODs _skeletalmeshcomponent) 1))
        (LocalVariables|SetLastMaterial (- (Rendering|Material|GetNumMaterials _skeletalmeshcomponent) 1))
        (switch int (Math|Conversions|ToInteger(Boolean) (Variables|PreviewSkelMesh|GetHideSkeletalMesh))
          (:Default
            ;; Hide Skeletal Mesh Loop
            (for _index (range (+ (LocalVariables|GetLastLOD) 1))
              (for _index_2 (range (+ (LocalVariables|GetLastMaterial) 1))
                (Components|SkinnedMesh|ShowMaterialSection (Variables|Default|GetSkeletalMeshComponent) _index_2 -1 false _index))))
          (:0
            ;; Hide Material Loop
            (for _index_3 (range (+ (LocalVariables|GetLastLOD) 1))
              (for _array_element (Variables|PreviewSkelMesh|GetHideMaterials)
                (Components|SkinnedMesh|ShowMaterialSection (Variables|Default|GetSkeletalMeshComponent) _array_element -1 false _index_3)))))))
    (:then_1
      ; Additional SkelMesh setup
      (for _array_element_2 (Variables|PreviewSkelMesh|GetAdditionalSkeletalMeshes)
        (LocalVariables|SetSkeletalMesh%20Local (Utilities|Struct|BreakBPClothPreviewSkelMeshStruct _array_element_2))
        (LocalVariables|SetHideMaterials%20Local (.HideMaterials_7_06180C3C4702B1B94380338BA44F531F (Utilities|Struct|BreakBPClothPreviewSkelMeshStruct _array_element_2)))
        (LocalVariables|SetHideSkelMesh%20Local (.HideSkeletalMesh_8_0AF0E42E409C0B71E1F732BC887E9ABA (Utilities|Struct|BreakBPClothPreviewSkelMeshStruct _array_element_2)))
        (if (Utilities|IsValid (LocalVariables|GetSkeletalMesh%20Local) :!class /Script/Engine.KismetSystemLibrary)
          (bind _returnvalue (AddComponent|NODE_AddSkeletalMeshComponent-0))
          (bind _output_get (LocalVariables|SetSkelMeshComp%20Local _returnvalue))
          (Utilities|Array|Add (Variables|Default|GetSkeletalMeshCompArray) _output_get)
          (Variables|/Script/Engine.SkeletalMeshComponent|SetSkeletalMeshAsset (LocalVariables|GetSkeletalMesh%20Local) _output_get)
          (Utilities|IsValid (.OverrideAnimationAsset_13_4DC02C6F4390972AD28A119C8F73F8C3 (Utilities|Struct|BreakBPClothPreviewSkelMeshStruct _array_element_2))
            (:Is%20Valid
              (Components|Animation|OverrideAnimationData (LocalVariables|GetSkelMeshComp%20Local) (.OverrideAnimationAsset_13_4DC02C6F4390972AD28A119C8F73F8C3 (Utilities|Struct|BreakBPClothPreviewSkelMeshStruct _array_element_2)) true true 0.0 1.0))
            (:Is%20Not%20Valid
              (Components|Animation|OverrideAnimationData (LocalVariables|GetSkelMeshComp%20Local) (Variables|PreviewSkelMesh|GetAnimationAsset) true true 0.0 1.0)))
          ;; Set Num LODs/Materials
          (LocalVariables|SetLastLOD (- (Components|SkinnedMesh|GetNumLODs _skelmeshcomp_local) 1))
          (LocalVariables|SetLastMaterial (- (Rendering|Material|GetNumMaterials _skelmeshcomp_local) 1))
          (switch int (Math|Conversions|ToInteger(Boolean) (LocalVariables|GetHideSkelMesh%20Local))
            (:Default
              ;; Hide Skeletal Mesh Loop
              (for _index_4 (range (+ (LocalVariables|GetLastLOD) 1))
                (for _index_5 (range (+ (LocalVariables|GetLastMaterial) 1))
                  (Components|SkinnedMesh|ShowMaterialSection (LocalVariables|GetSkelMeshComp%20Local) _index_5 -1 false _index_4))))
            (:0
              ;; Hide Material Loop
              (for _index_6 (range (+ (LocalVariables|GetLastLOD) 1))
                (for _array_element_3 (LocalVariables|GetHideMaterials%20Local)
                  (Components|SkinnedMesh|ShowMaterialSection (LocalVariables|GetSkelMeshComp%20Local) _array_element_3 -1 false _index_6))))))))))
